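; Blueprint actor for the Golden Fruit consumable in SunRose gameplay, with almost no logic of its own.
; BeginPlay and Tick are disabled placeholder events, and the construction script only calls its parent.
; The sound indicator check always returns false for any Fort player controller, and the actor audio parameters function returns a default value.
; Any consumable behavior is not in this graph.
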(event EventBeginPlay
  (|CallParentFunction) ; This node is disabled and will not be called.\nDrag off pins to build functionality.
)

(event EventTick
  (params (DeltaSeconds :type Float%20%28single-precision%29))
  (|CallParentFunction DeltaSeconds) ; This node is disabled and will not be called.\nDrag off pins to build functionality.
)

(fn GetActorAudioParameters
  (return 0))

(fn ShouldShowSoundIndicator
  (params (PlayerController :type Fort%20Player%20Controller%20Object%20Reference))
  (return false))

(fn ConstructionScript
  (|CallParentFunction))
