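; Golden Fruit consumable ability that shows an expiration warning gameplay cue on the player as the fruit's passive effect nears its end.
; When the Consuming tag is removed, it reads the passive effect's remaining duration and either starts the cue or sets a timer to recheck at the warning threshold.
; A parallel tag query wait removes the granting effect and ends the ability, which clears the timer and cue.
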(event Ability|EventFailedToActivatePassiveAbility
  (params (ActorInfo :type Gameplay%20Ability%20Actor%20Info%20Structure)))

(event Ability|EventActivateAbility
  ;[1 1 1] Ability Activate
  (bind _outfortplayerpawn (CallFunction|GetAvatarasValidFortPlayerPawn)
    (:ValidFortPlayerPawn
      (Variables|Default|SetPlayer%20Pawn _outfortplayerpawn)
      (Ability|Tasks|WaitGameplayTagRemove (Variables|Default|GetTag_Consuming)
        (:then
          (Ability|Tasks|WaitGameplayTagQuery (Variables|Default|GetTag%20Query%20Active)
            (:Triggered
              (Ability|RemoveGrantedbyEffect)
              (Ability|EndAbility))))
        (:Removed
          (CallFunction|UpdateExpirationWarningState))))))

(event Custom|UpdateExpirationWarningState
  (bind _returnvalue (CallFunction|RemainingGoldenFruitDuration))
  (bind _row_warning_duration (Variables|Default|GetRow%20Warning%20Duration))
  (bind _returnvalue_2 (ScalableFloat|GetValueatLevel _row_warning_duration))
  (bind _returnvalue_3 (- _returnvalue _returnvalue_2))
  ;[1 1 1] Expiration Warning State
  (if (> _returnvalue_3 0)
    (CallFunction|StopExpirationWarning)
    (bind _returnvalue_4 (Utilities|Time|SetTimerbyEvent (EventDispatchers|CreateEvent :!event Custom|UpdateExpirationWarningState) _returnvalue_3))
    (Variables|Default|SetExpiration%20Timer _returnvalue_4)
    (else
      (CallFunction|StartExpirationWarning))))

(event Custom|StartExpirationWarning
  ;[1 1 1] Expiration Warning State
  (if (not (Variables|Default|GetExpiration%20Warning%20Active))
    (Variables|Default|SetExpiration%20Warning%20Active true)
    (Ability|AddGameplayCueToOwner (Variables|Default|GetEnding%20Gameplay%20Cue%20Tag))))

(event Custom|StopExpirationWarning
  ;[1 1 1] Expiration Warning State
  (Utilities|Time|ClearandInvalidateTimerbyHandle (Variables|Default|GetExpiration%20Timer))
  (Ability|RemoveGameplayCueFromOwner (Variables|Default|GetEnding%20Gameplay%20Cue%20Tag))
  (if (Variables|Default|GetExpiration%20Warning%20Active)
    (Variables|Default|SetExpiration%20Warning%20Active false)))

(event Ability|EventOnEndAbility
  (params (bWasCancelled :type Boolean))
  ;[1 1 1] Clean up on event end
  (Utilities|Time|ClearandInvalidateTimerbyHandle (Variables|Default|GetExpiration%20Timer))
  (Ability|RemoveGameplayCueFromOwner (Variables|Default|GetEnding%20Gameplay%20Cue%20Tag)))

(fn RemainingGoldenFruitDuration
  (bind _returnvalue (GameplayEffects|GetActiveEffectswithAllTags (Ability|GetAbilitySystemComponent (Variables|Default|GetPlayer%20Pawn)) "(GameplayTags=((TagName=\"Asset.SunRose.GoldenFruit.Passive\")))"))
  (bind _returnvalue_2 (Ability|GameplayEffect|GetActiveGameplayEffectRemainingDuration (Utilities|Array|Get(aref) _returnvalue)))
  (return _returnvalue_2))
